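# Physics asset for the female medium RustyBolt body, covering six dynamic pouch bone chains around the pelvis and thighs.
# Each chain links master, pouch and end capsules with tight swing and twist limits; only the middle capsule is pouch-sized.
# The six masters and both thighs connect to the pelvis with free angular motion.
# The first chain's end bone is named dyn_pouch_end, without the number.
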PhysicsAsset /BRCosmetics/Characters/Player/Female/Medium/Bodies/F_MED_RustyBolt/Meshes/F_MED_RustyBolt_Physics
bodies {
 dyn_pouch1 [{
   shapeType=Capsule
   length=5.0547
   radius=3.1453
   shapeName=dyn_pouch1_capsule
   center=<0.1682 -0.4169 -0.1926>
   rotation=<rot 0 102.4061 -87.3709>
  }]
 dyn_pouch1_master [{shapeType=Capsule, length=0.505, radius=0.505, shapeName=dyn_pouch1_master_capsule, rotation=<rot 0 -90 0>}]
 dyn_pouch2 [{
   shapeType=Capsule
   length=5.0138
   radius=3.1395
   shapeName=dyn_pouch2_capsule
   center=<0.1454 -0.3833 -0.3648>
   rotation=<rot 0 100.2605 -82.6977>
  }]
 dyn_pouch2_end [{shapeType=Capsule, length=0.505, radius=0.505, shapeName=dyn_pouch2_end_capsule, rotation=<rot 0 -90 0>}]
 dyn_pouch2_master [{shapeType=Capsule, length=0.505, radius=0.505, shapeName=dyn_pouch2_master_capsule, rotation=<rot 0 -90 0>}]
 dyn_pouch3 [{
   shapeType=Capsule
   length=4.2227
   radius=5.9231
   shapeName=dyn_pouch3_capsule
   center=<0.8885 0.0091 -0.0961>
   rotation=<rot 0 90.0216 -89.97>
  }]
 dyn_pouch3_end [{shapeType=Capsule, length=0.505, radius=0.505, shapeName=dyn_pouch3_end_capsule, rotation=<rot 0 -90 0>}]
 dyn_pouch3_master [{shapeType=Capsule, length=0.505, radius=0.505, shapeName=dyn_pouch3_master_capsule, rotation=<rot 0 -90 0>}]
 dyn_pouch4 [{
   shapeType=Capsule
   length=6.7937
   radius=3.5861
   shapeName=dyn_pouch4_capsule
   center=<0.8926 0.0371 -0.0948>
   rotation=<rot 0 89.9017 -89.7796>
  }]
 dyn_pouch4_end [{shapeType=Capsule, length=0.505, radius=0.505, shapeName=dyn_pouch4_end_capsule, rotation=<rot 0 -90 0>}]
 dyn_pouch4_master [{shapeType=Capsule, length=0.505, radius=0.505, shapeName=dyn_pouch4_master_capsule, rotation=<rot 0 -90 0>}]
 dyn_pouch5 [{
   shapeType=Capsule
   length=7.4002
   radius=6.9099
   shapeName=dyn_pouch5_capsule
   center=<0.5736 -0.0532 0.0549>
   rotation=<rot -30.4051 -90 0.5518>
  }]
 dyn_pouch5_end [{shapeType=Capsule, length=0.505, radius=0.505, shapeName=dyn_pouch5_end_capsule, rotation=<rot 0 -90 0>}]
 dyn_pouch5_master [{shapeType=Capsule, length=0.505, radius=0.505, shapeName=dyn_pouch5_master_capsule, rotation=<rot 0 -90 0>}]
 dyn_pouch6 [{
   shapeType=Capsule
   length=6.7916
   radius=3.5144
   shapeName=dyn_pouch6_capsule
   center=<0.8874 -0.0657 0.0657>
   rotation=<rot 0 -89.7434 -89.7145>
  }]
 dyn_pouch6_end [{shapeType=Capsule, length=0.505, radius=0.505, shapeName=dyn_pouch6_end_capsule, rotation=<rot 0 -90 0>}]
 dyn_pouch6_master [{shapeType=Capsule, length=0.505, radius=0.505, shapeName=dyn_pouch6_master_capsule, rotation=<rot 0 -90 0>}]
 dyn_pouch_end [{shapeType=Capsule, length=0.505, radius=0.505, shapeName=dyn_pouch_end_capsule, rotation=<rot 0 -90 0>}]
 pelvis [{shapeType=Capsule, length=1, radius=1, shapeName=pelvis_capsule}]
 thigh_l [{shapeType=Capsule, length=24.5032, radius=2.7855, shapeName=thigh_l_capsule, center=<-14.2143 0.0001 0.0002>, rotation=<rot 90 -90 -79.9996>}]
 thigh_r [{shapeType=Capsule, length=24.2734, radius=9.9479, shapeName=thigh_r_capsule, center=<16.8632 0.0001 -0.0001>, rotation=<rot 90 0 0>}]
}
constraints [
 {bone1Name=dyn_pouch1, bone2Name=dyn_pouch1_master, swing1LimitDegrees=10, swing2LimitDegrees=10, twistLimitDegrees=10}
 {bone1Name=dyn_pouch_end, bone2Name=dyn_pouch1, swing1LimitDegrees=10, swing2LimitDegrees=10, twistLimitDegrees=10}
 {bone1Name=dyn_pouch2, bone2Name=dyn_pouch2_master, swing1LimitDegrees=10, swing2LimitDegrees=10, twistLimitDegrees=10}
 {bone1Name=dyn_pouch2_end, bone2Name=dyn_pouch2, swing1LimitDegrees=10, swing2LimitDegrees=10, twistLimitDegrees=10}
 {bone1Name=dyn_pouch3, bone2Name=dyn_pouch3_master, swing1LimitDegrees=10, swing2LimitDegrees=10, twistLimitDegrees=10}
 {bone1Name=dyn_pouch3_end, bone2Name=dyn_pouch3, swing1LimitDegrees=10, swing2LimitDegrees=10, twistLimitDegrees=10}
 {bone1Name=dyn_pouch4, bone2Name=dyn_pouch4_master, swing1LimitDegrees=10, swing2LimitDegrees=10, twistLimitDegrees=10}
 {bone1Name=dyn_pouch4_end, bone2Name=dyn_pouch4, swing1LimitDegrees=10, swing2LimitDegrees=10, twistLimitDegrees=10}
 {bone1Name=dyn_pouch5, bone2Name=dyn_pouch5_master, swing1LimitDegrees=10, swing2LimitDegrees=10, twistLimitDegrees=10}
 {bone1Name=dyn_pouch5_end, bone2Name=dyn_pouch5, swing1LimitDegrees=10, swing2LimitDegrees=10, twistLimitDegrees=10}
 {bone1Name=dyn_pouch6, bone2Name=dyn_pouch6_master, swing1LimitDegrees=10, swing2LimitDegrees=10, twistLimitDegrees=10}
 {bone1Name=dyn_pouch6_end, bone2Name=dyn_pouch6, swing1LimitDegrees=10, swing2LimitDegrees=10, twistLimitDegrees=10}
 {
  bone1Name=dyn_pouch1_master
  bone2Name=pelvis
  swing1LimitDegrees=45
  swing1Motion=Free
  swing2LimitDegrees=45
  swing2Motion=Free
  twistLimitDegrees=45
  twistMotion=Free
 }
 {
  bone1Name=dyn_pouch2_master
  bone2Name=pelvis
  swing1LimitDegrees=45
  swing1Motion=Free
  swing2LimitDegrees=45
  swing2Motion=Free
  twistLimitDegrees=45
  twistMotion=Free
 }
 {
  bone1Name=dyn_pouch3_master
  bone2Name=pelvis
  swing1LimitDegrees=45
  swing1Motion=Free
  swing2LimitDegrees=45
  swing2Motion=Free
  twistLimitDegrees=45
  twistMotion=Free
 }
 {
  bone1Name=dyn_pouch4_master
  bone2Name=pelvis
  swing1LimitDegrees=45
  swing1Motion=Free
  swing2LimitDegrees=45
  swing2Motion=Free
  twistLimitDegrees=45
  twistMotion=Free
 }
 {
  bone1Name=dyn_pouch5_master
  bone2Name=pelvis
  swing1LimitDegrees=45
  swing1Motion=Free
  swing2LimitDegrees=45
  swing2Motion=Free
  twistLimitDegrees=45
  twistMotion=Free
 }
 {
  bone1Name=dyn_pouch6_master
  bone2Name=pelvis
  swing1LimitDegrees=45
  swing1Motion=Free
  swing2LimitDegrees=45
  swing2Motion=Free
  twistLimitDegrees=45
  twistMotion=Free
 }
 {
  bone1Name=thigh_l
  bone2Name=pelvis
  swing1LimitDegrees=45
  swing1Motion=Free
  swing2LimitDegrees=45
  swing2Motion=Free
  twistLimitDegrees=45
  twistMotion=Free
 }
 {
  bone1Name=thigh_r
  bone2Name=pelvis
  swing1LimitDegrees=45
  swing1Motion=Free
  swing2LimitDegrees=45
  swing2Motion=Free
  twistLimitDegrees=45
  twistMotion=Free
 }
]
