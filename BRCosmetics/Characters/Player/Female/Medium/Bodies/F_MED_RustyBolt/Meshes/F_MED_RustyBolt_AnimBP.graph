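# Post-processing AnimBP for the RustyBolt character body that corrects pouches, shoulder pads, fur collar and boot over the parent mesh's copied pose.
# A RigidBody simulation feeds constraint, look-at and bone-transform nodes, with crouch and parachute toggles.
# The update event converts thigh, arm and head rotations relative to reference pose into corrective alphas, zeroing some by movement or emote state.
# A Control Rig node sits unconnected.
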AnimBlueprint /BRCosmetics/Characters/Player/Female/Medium/Bodies/F_MED_RustyBolt/Meshes/F_MED_RustyBolt_AnimBP
target_skeleton=&/F_MED_RustyBolt_Skeleton
graphs [
 AnimGraph {nodes [
   Control Rig() {
    alphaBoolBlend {blendOption=Linear}
    controlRigAssetReference {blueprintRigClass=&/F_MED_RustyBolt_CtrlRig.F_MED_RustyBolt_CtrlRig_C}
   }
   n1 = RigidBody(ComponentPose=Local To Component(LocalPose=Copy Pose From Mesh() {bUseAttachedParent=true})) {
    bOverrideWorldGravity=true
    overridePhysicsAsset=&/F_MED_RustyBolt_Physics
    simulationSpace=BaseBoneSpace
    alphaBoolBlend {blendOption=Linear}
    baseBoneRef {boneName=pelvis}
    componentAppliedLinearAccClamp=<2000 2000 2000>
    componentLinearAccScale=<0.3 0.3 0.35>
    componentLinearVelScale=<0.3 0.3 0.35>
    overrideWorldGravity {z=-980}
   }
   n2 = "Constraint - dyn_pouch1_master (P)"(ComponentPose=@n1, Alpha=0.5, ConstraintWeights_0=1) {
    alphaBoolBlend {blendOption=Linear}
    boneToModify {boneName=dyn_pouch1_master}
    constraintSetup [{targetBone {boneName=thigh_twist_01_l}}]
    constraintWeights [1]
   }
   n3 = "Constraint - dyn_pouch2_master (P)"(ComponentPose=@n2, Alpha=0.5, ConstraintWeights_0=1) {
    alphaBoolBlend {blendOption=Linear}
    boneToModify {boneName=dyn_pouch2_master}
    constraintSetup [{targetBone {boneName=thigh_twist_01_l}}]
    constraintWeights [1]
   }
   n4 = Get bIsCrouching()
   n5 = "Transform (Modify) Bone - Bone: dyn_pouch1_master"!(ComponentPose=@n3, bAlphaBoolEnabled:bool=@n4, Translation="0,0,0", Rotation="0,0,0", Scale="0,0,0") -> (Pose) {
    alphaInputType=Bool
    scaleMode=BMM_Replace
    scaleSpace=BCS_BoneSpace
    alphaBoolBlend {blendOption=Linear}
    boneToModify {boneName=dyn_pouch1_master}
    scale=<0 0 0>
   }
   n6 = "Transform (Modify) Bone - Bone: dyn_pouch2_master"!(ComponentPose=@n5, bAlphaBoolEnabled:bool=@n4, Translation="0,0,0", Rotation="0,0,0", Scale="0,0,0") -> (Pose) {
    alphaInputType=Bool
    scaleMode=BMM_Replace
    scaleSpace=BCS_BoneSpace
    alphaBoolBlend {blendOption=Linear}
    boneToModify {boneName=dyn_pouch2_master}
    scale=<0 0 0>
   }
   n7 = "Transform (Modify) Bone - Bone: dyn_pouch6_master"!(ComponentPose=@n6, bAlphaBoolEnabled:bool=Get bIsCrouching(), Translation="0,0,0", Rotation="-30.000,0.000,0.000", Scale="1,1,1") -> (Pose) {
    alphaInputType=Bool
    rotationMode=BMM_Additive
    alphaBoolBlend {blendInTime=0.5, blendOption=Linear, blendOutTime=0.5}
    boneToModify {boneName=dyn_pouch6_master}
    rotation {pitch=-30}
   }
   n8 = "Constraint - dyn_pouch6_master (P)"(ComponentPose=@n7, Alpha=0.5, ConstraintWeights_0=1) {
    alphaBoolBlend {blendOption=Linear}
    boneToModify {boneName=dyn_pouch6_master}
    constraintSetup [{targetBone {boneName=thigh_twist_01_r}}]
    constraintWeights [1]
   }
   n9 = "Transform (Modify) Bone - Bone: dyn_pouch6_master"!(
    ComponentPose=@n8
    Alpha:float=Get Thigh R Crouch Rot Alpha() -> (Thigh R Crouch Rot Alpha:double)
    Translation="-0.707,5.786,2.140"
    Rotation="-20.000,-0.000,-10.000"
    Scale="1,1,1"
   ) -> (Pose) {
    rotationMode=BMM_Additive
    translationMode=BMM_Additive
    alphaBoolBlend {blendInTime=0.5, blendOption=Linear, blendOutTime=0.5}
    boneToModify {boneName=dyn_pouch6_master}
    rotation {pitch=-20, roll=-10}
    translation=<-0.707 5.786 2.14>
   }
   n10 = "Transform (Modify) Bone - Bone: dyn_pouch6_master"(
    ComponentPose=@n9
    Alpha=Get Thigh R Pitch Alpha() -> (Thigh R Pitch Alpha:double)
    Translation="-4.367,1.599,0.144"
    Rotation="-12.807,-20.208,-5.963"
    Scale="1,1,1"
   ) {
    rotationMode=BMM_Additive
    translationMode=BMM_Additive
    alphaBoolBlend {blendInTime=0.5, blendOption=Linear, blendOutTime=0.5}
    alphaScaleBiasClamp {bInterpResult=true}
    boneToModify {boneName=dyn_pouch6_master}
    rotation=<rot -12.807 -20.208 -5.9631>
    translation=<-4.367 1.599 0.144>
   }
   n11 = "Transform (Modify) Bone - Bone: dyn_pouch6"(
    ComponentPose=@n10
    Alpha=Get Jump Alpha() -> (Jump Alpha:double)
    Translation="-0.318,4.530,-2.812"
    Rotation="-24.507,59.398,-33.491"
    Scale="1,1,1"
   ) {
    rotationMode=BMM_Additive
    rotationSpace=BCS_BoneSpace
    translationMode=BMM_Additive
    translationSpace=BCS_BoneSpace
    alphaBoolBlend {blendInTime=0.5, blendOption=Linear, blendOutTime=0.5}
    alphaScaleBiasClamp {interpSpeedDecreasing=15}
    boneToModify {boneName=dyn_pouch6}
    rotation=<rot -24.507 59.398 -33.491>
    translation=<-0.318 4.53 -2.812>
   }
   n12 = Look At - Bone: shoulderpads_clav_r!(ComponentPose=@n11, bAlphaBoolEnabled:bool=True) -> (Pose) {
    alphaInputType=Bool
    alphaBoolBlend {blendOption=Linear}
    boneToModify {boneName=shoulderpads_clav_r}
    lookAtLocation {x=0, z=-10}
    lookAtTarget {boneReference {boneName=lowerarm_r}}
    lookAt_Axis {axis=<-1 0>}
   }
   n13 = Look At - Bone: shoulderpads_clav_l!(ComponentPose=@n12, bAlphaBoolEnabled:bool=True) -> (Pose) {
    alphaInputType=Bool
    alphaBoolBlend {blendOption=Linear}
    boneToModify {boneName=shoulderpads_clav_l}
    lookAtLocation {x=0, z=10}
    lookAtTarget {boneReference {boneName=lowerarm_l}}
    lookAt_Axis {axis=<1 0>}
   }
   n14 = "Transform (Modify) Bone - Bone: shoulderpads_clav_r"(
    ComponentPose=@n13
    Alpha=Get armRestRightAlpha() -> (armRestRightAlpha:double)
    Translation="0,0,0"
    Rotation="0.000,15,0.000"
    Scale="1,1,1"
   ) {
    rotationMode=BMM_Additive
    rotationSpace=BCS_BoneSpace
    alphaBoolBlend {blendOption=Linear}
    boneToModify {boneName=shoulderpads_clav_r}
    rotation {yaw=15}
   }
   n15 = "Transform (Modify) Bone - Bone: shoulderpads_clav_r"(ComponentPose=@n14, Alpha=1, Translation="0,0,0", Rotation="0,-10,10", Scale="1,1,1") {
    rotationMode=BMM_Additive
    rotationSpace=BCS_BoneSpace
    alphaBoolBlend {blendOption=Linear}
    boneToModify {boneName=shoulderpads_clav_r}
    rotation {roll=10, yaw=-10}
   }
   n16 = "Transform (Modify) Bone - Bone: shoulderpads_clav_r"(ComponentPose=@n15, Alpha=Get armUpRightAlpha() -> (armUpRightAlpha:double), Translation="0,0,0", Rotation="0,10,-10", Scale="1,1,1") {
    rotationMode=BMM_Additive
    rotationSpace=BCS_BoneSpace
    alphaBoolBlend {blendOption=Linear}
    boneToModify {boneName=shoulderpads_clav_r}
    rotation {roll=-10, yaw=10}
   }
   n17 = "Transform (Modify) Bone - Bone: shoulderpads_clav_l"(ComponentPose=@n16, Alpha=Get armUpLeftAlpha() -> (armUpLeftAlpha:double), Translation="0,0,0", Rotation="0.000,8,0.000", Scale="1,1,1") {
    rotationMode=BMM_Additive
    rotationSpace=BCS_BoneSpace
    alphaBoolBlend {blendOption=Linear}
    boneToModify {boneName=shoulderpads_clav_l}
    rotation {yaw=8}
   }
   n18 = "Transform (Modify) Bone - Bone: shoulderpads_clav_r"(
    ComponentPose=@n17
    Alpha=Get armRollRightAlpha() -> (armRollRightAlpha:double)
    Translation="0,0,0"
    Rotation="0.000,40.000,0.000"
    Scale="1,1,1"
   ) {
    rotationMode=BMM_Additive
    rotationSpace=BCS_BoneSpace
    alphaBoolBlend {blendOption=Linear}
    boneToModify {boneName=shoulderpads_clav_r}
    rotation {yaw=39.9999}
   }
   n19 = "Transform (Modify) Bone - Bone: shoulderpads_clav_l"(ComponentPose=@n18, Alpha=Get armRollLeftAlpha() -> (armRollLeftAlpha:double), Translation="0,0,0", Rotation="0.000,20,0.000", Scale="1,1,1") {
    rotationMode=BMM_Additive
    rotationSpace=BCS_BoneSpace
    alphaBoolBlend {blendOption=Linear}
    boneToModify {boneName=shoulderpads_clav_l}
    rotation {yaw=20}
   }
   n20 = "Transform (Modify) Bone - Bone: shoulderpads_clav_l"!(ComponentPose=@n19, bAlphaBoolEnabled:bool=Get bIsParachuteOpened(), Translation="3,0,0", Rotation="20,30,0", Scale="1,1,1") -> (Pose) {
    alphaInputType=Bool
    rotationMode=BMM_Additive
    rotationSpace=BCS_BoneSpace
    translationMode=BMM_Additive
    translationSpace=BCS_BoneSpace
    alphaBoolBlend {blendInTime=0.5, blendOption=Linear, blendOutTime=0.5}
    alphaScaleBiasClamp {bInterpResult=true}
    boneToModify {boneName=shoulderpads_clav_l}
    rotation {pitch=20, yaw=30}
    translation {x=3}
   }
   n21 = "Transform (Modify) Bone - Bone: shoulderpads_clav_r"!(ComponentPose=@n20, bAlphaBoolEnabled:bool=Get bIsParachuteOpened(), Translation="-3,0,0", Rotation="20,30,0", Scale="1,1,1") -> (Pose) {
    alphaInputType=Bool
    rotationMode=BMM_Additive
    rotationSpace=BCS_BoneSpace
    translationMode=BMM_Additive
    translationSpace=BCS_BoneSpace
    alphaBoolBlend {blendInTime=0.5, blendOption=Linear, blendOutTime=0.5}
    alphaScaleBiasClamp {bInterpResult=true}
    boneToModify {boneName=shoulderpads_clav_r}
    rotation {pitch=20, yaw=30}
    translation {x=-3}
   }
   n22 = "Transform (Modify) Bone - Bone: fur_collar"(ComponentPose=@n21, Alpha=Get neckAlpha() -> (neckAlpha:double), Translation="0,0,0", Rotation="10.000,-10.000,0.000", Scale="1,1,1") {
    rotationMode=BMM_Additive
    rotationSpace=BCS_BoneSpace
    alphaBoolBlend {blendOption=Linear}
    boneToModify {boneName=fur_collar}
    rotation=<rot 10 -10.0002 0>
   }
   n23 = "Transform (Modify) Bone - Bone: fur_boot_l"!(ComponentPose=@n22, bAlphaBoolEnabled:bool=Get bIsCrouching(), Translation="0,0,0", Rotation="0.000,-35,0.000", Scale="1,1,1") -> (Pose) {
    alphaInputType=Bool
    rotationMode=BMM_Additive
    rotationSpace=BCS_BoneSpace
    alphaBoolBlend {blendOption=Linear}
    boneToModify {boneName=fur_boot_l}
    rotation {yaw=-35}
   }
   Output Pose(Result=Component To Local(ComponentPose=@n23)) {layerGroup=None}
  ]}
 EventGraph {nodes [
   Event BlueprintUpdateAnimation()
   > n1 = Sequence!(execute) -> (then_0, then_1, then_2:Exec, then_3:Exec)
   > n2 = Branch(Condition=Get Jump Falling() -> (Jump Falling:bool))
   n3 = BreakTransform(InTransform=GetDeltaTransformFromRefPose(self=GetOwnerSkeletalMeshComponent(), BoneName=thigh_r, BaseName=pelvis))
   > Set Jump Alpha(Jump Alpha:double=MapRangeClamped(Value=BreakRotator(InRot=@n3.Rotation).Yaw, InRangeA=10, InRangeB=-100)) -> (Output_Get:double)
   Set Jump Alpha(execute=@n2.else, Jump Alpha=0)
   n4 = Branch(execute=@n1.then_1, Condition=Get bIsCrouching())
   n5 = GetOwnerSkeletalMeshComponent()
   n6 = BreakRotator(InRot=BreakTransform(InTransform=GetDeltaTransformFromRefPose(self=@n5, BoneName=thigh_r, BaseName=pelvis)).Rotation)
   > Set Thigh R Crouch Rot Alpha(Thigh R Crouch Rot Alpha:double=MapRangeClamped(Value=@n6.Yaw, InRangeA=-14, InRangeB=-105)) -> (Output_Get:double)
   > Set Thigh R Pitch Alpha(Thigh R Pitch Alpha:double=MapRangeClamped(Value=@n6.Pitch, InRangeA=15, InRangeB=30)) -> (Output_Get:double)
   Set Thigh R Crouch Rot Alpha(execute=@n4.else, Thigh R Crouch Rot Alpha=0)
   > Set Thigh R Pitch Alpha(Thigh R Pitch Alpha=0)
   n7 = GetOwnerSkeletalMeshComponent()
   n8 = BreakRotator(InRot=BreakTransform(InTransform=GetDeltaTransformFromRefPose(self=@n7, BoneName=head, BaseName=spine_05)).Rotation)
   Set aimUpAlpha(execute=@n1.then_2, aimUpAlpha:double=MapRangeClamped(Value=@n8.Pitch, InRangeA=20, InRangeB=45)) -> (Output_Get:double)
   > Set neckAlpha(neckAlpha:double=MapRangeClamped(Value=@n8.Pitch, InRangeA=0, InRangeB=45)) -> (Output_Get:double)
   n9 = GetOwnerSkeletalMeshComponent()
   n10 = BreakRotator(InRot=BreakTransform(InTransform=GetDeltaTransformFromRefPose(self=@n9, BoneName=upperarm_r, BaseName=clavicle_r)).Rotation)
   > Set armUpRightAlpha(armUpRightAlpha:double=MapRangeClamped(Value=@n10.Yaw, InRangeA=30, InRangeB=115)) -> (Output_Get:double)
   > Set armRestRightAlpha(armRestRightAlpha:double=MapRangeClamped(Value=@n10.Yaw, InRangeA=30, InRangeB=-7)) -> (Output_Get:double)
   > Set armRollRightAlpha(armRollRightAlpha:double=MapRangeClamped(Value=@n10, InRangeA=0, InRangeB=-70)) -> (Output_Get:double)
   n11 = GetOwnerSkeletalMeshComponent()
   n12 = BreakRotator(InRot=BreakTransform(InTransform=GetDeltaTransformFromRefPose(self=@n11, BoneName=upperarm_l, BaseName=clavicle_l)).Rotation)
   > Set armUpLeftAlpha(armUpLeftAlpha:double=MapRangeClamped(Value=@n12.Pitch, InRangeA=0, InRangeB=40)) -> (Output_Get:double)
   > Set armYawLeftAlpha(armYawLeftAlpha:double=MapRangeClamped(Value=@n12.Yaw, InRangeA=15, InRangeB=60)) -> (Output_Get:double)
   > Set armRollLeftAlpha(armRollLeftAlpha:double=MapRangeClamped(Value=@n12, InRangeA=-70, InRangeB=-100)) -> (Output_Get:double)
   Set Jump Falling(execute=@n1.then_3, Jump Falling:bool=OR Boolean(A=Get bIsJumping(), B=Get bIsFalling())) -> (Output_Get:bool)
   > n13 = Set Skydiving(Skydiving:bool=OR Boolean!(A=Get bIsSkyDiving(), B=Get bIsSkydiveFloating(), C:bool=Get bIsSkydiveDiveMode()) -> (ReturnValue)) -> (Output_Get:bool)
   > Branch(Condition=OR Boolean!(A=@n13, B=Get bIsPlayingEmote()) -> (ReturnValue))
   > Set aimUpAlpha(aimUpAlpha=0)
   > Set neckAlpha(neckAlpha=0)
   > Set armRestRightAlpha(armRestRightAlpha=0)
   > Set armYawLeftAlpha(armYawLeftAlpha=0)
   > Set armUpLeftAlpha(armUpLeftAlpha=0)
   > Set armUpRightAlpha(armUpRightAlpha=0)
   > Branch(Condition=@n13)
   > Set armRollLeftAlpha(armRollLeftAlpha=0)
   > Set armRollRightAlpha(armRollRightAlpha=0)
  ]}
]
variables [
 aimUpAlpha {type=number}
 armRestRightAlpha {type=number}
 armRollLeftAlpha {type=number}
 armRollRightAlpha {type=number}
 armUpLeftAlpha {type=number}
 armUpRightAlpha {type=number}
 armYawLeftAlpha {type=number}
 Jump Alpha {type=number}
 Jump Falling {type=boolean}
 neckAlpha {type=number}
 Skydiving {type=boolean}
 Thigh R Crouch Rot Alpha {type=number}
 Thigh R Pitch Alpha {type=number}
]
