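# Physics asset for the female medium RustyBolt character's face accessory mesh, setting up physics for front hair and ponytail bones.
# Capsule bodies form a four-link front hair chain and a five-link ponytail chain, each from a small master body, plus a head capsule.
# Small angular limits join the links, with the ponytail tightest mid-chain and loosest toward the tip.
# The ponytail master joins the head with free rotation, while the front hair master has no head constraint.
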PhysicsAsset /BRCosmetics/Characters/Player/Female/Medium/Bodies/F_MED_RustyBolt/Meshes/Parts/F_MED_RustyBolt_FaceAcc_Physics
bodies {
 dyn_fr_hair_1 [{shapeType=Capsule, length=6.3782, radius=2.1495, shapeName=dyn_fr_hair_1_capsule, center=<2.4793 -0.0002 0>, rotation=<rot 0 0 89.9999>}]
 dyn_fr_hair_2 [{shapeType=Capsule, length=4.8136, radius=1.7308, shapeName=dyn_fr_hair_2_capsule, center=<2.4793 -0.0002 0>, rotation=<rot 0 0 89.9999>}]
 dyn_fr_hair_3 [{shapeType=Capsule, length=4.7241, radius=1.9031, shapeName=dyn_fr_hair_3_capsule, center=<2.4793 -0.0002 0>, rotation=<rot 0 0 89.9999>}]
 dyn_fr_hair_4 [{shapeType=Capsule, length=3.252, radius=1.4942, shapeName=dyn_fr_hair_4_capsule, center=<0.5556 0 -1.991>, rotation=<rot 0 0 89.9999>}]
 dyn_fr_hair_master [{shapeType=Capsule, length=0.505, radius=0.505, shapeName=dyn_fr_hair_master_capsule, rotation=<rot 0 -90 0>}]
 dyn_ponytail_1 [{shapeType=Capsule, length=3, radius=2, shapeName=dyn_ponytail_1_capsule, center=<2.7031 0.0002 0>, rotation=<rot 90 0 0>}]
 dyn_ponytail_2 [{shapeType=Capsule, length=3, radius=2, shapeName=dyn_ponytail_2_capsule, center=<2.7031 0.0002 0>, rotation=<rot 90 0 0>}]
 dyn_ponytail_3 [{shapeType=Capsule, length=3, radius=2, shapeName=dyn_ponytail_3_capsule, center=<2.7031 0.0002 0>, rotation=<rot 90 0 0>}]
 dyn_ponytail_4 [{shapeType=Capsule, length=3, radius=2, shapeName=dyn_ponytail_4_capsule, center=<2.7031 0.0002 0>, rotation=<rot 90 0 0>}]
 dyn_ponytail_5 [{shapeType=Capsule, length=3, radius=2, shapeName=dyn_ponytail_5_capsule, center=<2.7031 0.0002 0>, rotation=<rot 90 0 0>}]
 dyn_ponytail_master [{shapeType=Capsule, length=0.505, radius=0.505, shapeName=dyn_ponytail_master_capsule, rotation=<rot 0 -90 0>}]
 head [{shapeType=Capsule, length=4.2884, radius=4.3717, shapeName=head_capsule, center=<1.6491 3.2671 0>, rotation=<rot 90 0 -10.0001>}]
}
constraints [
 {bone1Name=dyn_fr_hair_1, bone2Name=dyn_fr_hair_master, swing1LimitDegrees=2, swing2LimitDegrees=10, twistLimitDegrees=5}
 {bone1Name=dyn_fr_hair_2, bone2Name=dyn_fr_hair_1, swing1LimitDegrees=2, swing2LimitDegrees=10, twistLimitDegrees=5}
 {bone1Name=dyn_fr_hair_3, bone2Name=dyn_fr_hair_2, swing1LimitDegrees=2, swing2LimitDegrees=10, twistLimitDegrees=5}
 {bone1Name=dyn_fr_hair_4, bone2Name=dyn_fr_hair_3, swing1LimitDegrees=2, swing2LimitDegrees=10, twistLimitDegrees=5}
 {bone1Name=dyn_ponytail_1, bone2Name=dyn_ponytail_master, swing1LimitDegrees=10, swing2LimitDegrees=10, twistLimitDegrees=5}
 {bone1Name=dyn_ponytail_2, bone2Name=dyn_ponytail_1, swing1LimitDegrees=10, swing2LimitDegrees=10, twistLimitDegrees=5}
 {bone1Name=dyn_ponytail_3, bone2Name=dyn_ponytail_2, swing1LimitDegrees=5, swing2LimitDegrees=5, twistLimitDegrees=2}
 {bone1Name=dyn_ponytail_4, bone2Name=dyn_ponytail_3, swing1LimitDegrees=20, swing2LimitDegrees=20, twistLimitDegrees=5}
 {bone1Name=dyn_ponytail_5, bone2Name=dyn_ponytail_4, swing1LimitDegrees=20, swing2LimitDegrees=20, twistLimitDegrees=5}
 {
  bone1Name=dyn_ponytail_master
  bone2Name=head
  swing1LimitDegrees=45
  swing1Motion=Free
  swing2LimitDegrees=45
  swing2Motion=Free
  twistLimitDegrees=45
  twistMotion=Free
 }
]
